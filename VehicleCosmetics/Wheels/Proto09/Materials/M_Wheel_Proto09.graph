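# Masked material for the Proto09 vehicle wheel, usable on skeletal meshes and mesh deformers.
# MF_WheelParent_Simple supplies the base attributes, whose normal is overridden before passing through the shared vehicle damage and de-rez transition functions.
# A TireRubberExist static switch adds a tiled matte detail normal to the normal, with named reroutes carrying the base normal, vertex-color red channel and rubber result.
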Material /VehicleCosmetics/Wheels/Proto09/Materials/M_Wheel_Proto09
graph {
 properties {
  bAutomaticallySetUsageInEditor=false
  bUseMaterialAttributes=true
  bUsedWithMeshDeformer=true
  bUsedWithSkeletalMesh=true
  blendMode=BLEND_Masked
  pixelDepthOffsetMode=PDOM_Legacy
  usedShadingModels=Default Lit
 }
 nodes [
  n1 = MaterialFunctionCall() {materialFunction=&MF_WheelParent_Simple}
  n2 = NamedRerouteUsage()
  MaterialFunctionCall(In=MaterialFunctionCall(Material=SetMaterialAttributes!(MaterialAttributes=@n1.MaterialAttributes, Normal=@n2) {attributeSetTypes [#1]}) {materialFunction=&MF_BaseVehicle_Damage}.Result) {materialFunction=&MF_VehicleDeRez_Transition}
  NamedRerouteDeclaration(None=GetMaterialAttributes(None=@n1.MaterialAttributes) {attributeGetTypes [#1]}.Normal) {name=NormalIN, nodeColor=<rgba 0 0.3437 1 1>}
  NamedRerouteDeclaration(None=VertexColor().R) {name=VC_R, nodeColor=<rgba 1 0 0.065 1>}
  n3 = NamedRerouteUsage()
  n4 = TextureCoordinate()
  n5 = ScalarParameter() {defaultValue=2, group=TireRubber, parameterName=TextureDetailNormalTiling}
  n6 = Multiply(
   A=TextureSample(UVs=Multiply(A=@n4, B=@n5)) {samplerType=SAMPLERTYPE_Normal, texture=/VehicleCosmetics/SharedTextures/T_Matte_N}.RGB
   B=NamedRerouteUsage()
  )
  NamedRerouteDeclaration(None=StaticSwitchParameter(True=Add(A=@n3, B=@n6), False=@n3) {group=TireRubber, parameterName=TireRubberExist}) {name=NormalRubber, nodeColor=<rgba 0.1016 0 1 1>}
 ]
}
