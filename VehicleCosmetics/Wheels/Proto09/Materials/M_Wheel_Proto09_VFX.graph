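# Translucent unlit material for the animated swirl VFX on the Proto09 esports vehicle wheel, usable on skeletal meshes.
# Squared distance from UV center drives panned rainbow-curve lookups, which combine with two counter-rotating twirl texture samples to form pulses and boosted highlights.
# A radial falloff confines the effect to a disc and, with the highlights, builds hardened, clamped opacity.
# PaintColor, EmissivePower and Zomba parameters control emissive tint, background level and intensity.
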Material /VehicleCosmetics/Wheels/Proto09/Materials/M_Wheel_Proto09_VFX
graph {
 properties {
  bAutomaticallySetUsageInEditor=false
  bCastRayTracedShadows=false
  bUsedWithMeshDeformer=true
  bUsedWithSkeletalMesh=true
  blendMode=BLEND_Translucent
  shadingModel=MSM_Unlit
  usedShadingModels=Unlit
 }
 nodes [
  n1 = Multiply(
   A=VectorParameter() {group=Zomba, parameterName=PaintColor, defaultValue=<rgba 1 1 1 1>}.RGB
   B=ScalarParameter() {defaultValue=2, parameterName=EmissivePower}
  )
  n2 = Add(A=TextureCoordinate(), B=Constant() {r=-0.5})
  n3 = Abs(None=DotProduct(A=@n2, B=@n2))
  n4 = Multiply(A=Clamp(None=@n3), B=Constant() {r=2})
  n5 = TextureSample(UVs=Panner(Coordinate=@n4) {speedX=-0.2}) {texture=/VehicleCosmetics/Wheels/_SharedTextures/FX/Curve_Rainbow}
  n6 = Panner(Coordinate=@n4) {speedX=-0.5}
  n7 = Add(A=Add(A=@n5.G, B=TextureSample(UVs=@n6) {texture=&Curve_Rainbow}.R), B=@n5.B)
  n8 = Constant() {r=0.8}
  n9 = TextureCoordinate()
  n10 = TextureSample(UVs=Rotator(Coordinate=@n9) {speed=1}) {texture=/VehicleCosmetics/Wheels/Proto09_Esports/Textures/T_Twirl_RGB_T}
  n11 = TextureSample(UVs=Rotator(Coordinate=@n9) {speed=-2}) {texture=&T_Twirl_RGB_T}
  n12 = Add(A=@n10.R, B=@n11.G)
  n13 = LinearInterpolate(A=@n12, B=Multiply(A=Multiply(A=@n7, B=@n8), B=@n12), Alpha=Constant() {r=0.5})
  n14 = Multiply(A=@n12, B=Multiply(A=@n10.B, B=@n11.A))
  n15 = Multiply(A=@n5.R, B=Clamp(None=ConstantBiasScale(None=@n14) {bias=-0.1, scale=20}))
  n16 = Multiply(A=Multiply(A=@n15, B=Constant() {r=1}), B=Constant() {r=6})
  n17 = Add(A=Add(A=Clamp(None=@n13), B=@n16), B=Multiply(A=Constant() {r=1}, B=Constant()))
  n18 = OneMinus(None=@n3)
  n19 = Constant() {r=-0.78}
  n20 = Clamp(None=Multiply(A=Add(A=@n18, B=@n19), B=Constant() {r=3}))
  n21 = Power(Base=@n20, Exp=Constant() {r=0.2})
  n22 = Multiply(A=@n1, B=Add(A=ScalarParameter() {group=Zomba, parameterName=ZombaBackgroundClamp}, B=Multiply(A=@n17, B=@n21)))
  n23 = ScalarParameter() {defaultValue=4, group=Zomba, parameterName=ZombaIntensity}
  NamedRerouteDeclaration(None=Add(A=Multiply(A=@n22, B=@n23), B=Multiply(A=@n1, B=@n16))) {name=Emissive, nodeColor=<rgba 0.8984 0 1 1>}
  n24 = ScalarParameter() {defaultValue=15, group=Zomba, parameterName=OpacityHardness}
  n25 = Multiply(A=@n20, B=Add(A=@n20, B=@n15))
  NamedRerouteDeclaration(None=Clamp(None=Multiply(A=@n24, B=@n25)) {maxDefault=0.9}) {name=Opacity, nodeColor=<rgba 0.5781 0.5781 0.5781 1>}
  NamedRerouteUsage()
  NamedRerouteUsage()
 ]
}
